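# Unlit additive emissive material for the PinkSpike glider effects, usable on Niagara mesh particles.
# It adds a tinted, contrast-boosted greyscale copy of the emissive texture, masked by a powered inverse Fresnel to favor camera-facing surfaces, onto the original texture.
# Particle color, alpha and an opacity parameter scale the result before a shared FX emissive adjustment function.
# A CameraOffset function supplies a parameterized camera-relative position offset.
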Material /BRCosmetics/Gadgets/Assets/VinderTech_GliderChute/Glider_PinkSpike/FX/M_Glider_Pinkspike_Emissive
graph {
 properties {
  bUsedWithNiagaraMeshParticles=true
  blendMode=BLEND_Additive
  pixelDepthOffsetMode=PDOM_Legacy
  shadingModel=MSM_Unlit
  usedShadingModels=Unlit
 }
 nodes [
  MaterialFunctionCall(Offset Amount=ScalarParameter() {defaultValue=10, parameterName=Camera Offset}) {materialFunction=/Engine/Functions/Engine_MaterialFunctions02/WorldPositionOffset/CameraOffset}
  n1 = ParticleColor()
  n2 = TextureSample() {texture=/BRCosmetics/Gadgets/Assets/VinderTech_GliderChute/Glider_PinkSpike/Textures/T_Glider_PinkSpike_E}
  n3 = Desaturation(None=@n2.RGB)
  n4 = Constant() {r=3}
  n5 = OneMinus(None=Fresnel() {baseReflectFraction=0, exponent=1})
  n6 = Multiply(
   A=Multiply(
    A=MaterialFunctionCall(In=@n3, Contrast=@n4) {materialFunction=&CheapContrast}.Result
    B=VectorParameter() {parameterName=Lines Tint, defaultValue=<rgba 1 1 1 1>}.RGB
   )
   B=Saturate(None=Power(Base=@n5) {constExponent=3})
  )
  n7 = ScalarParameter() {defaultValue=1, parameterName=Lines Intensity}
  n8 = Multiply(A=@n1.RGB, B=Add(A=@n2.RGB, B=Multiply(A=@n6, B=@n7)))
  MaterialFunctionCall(Emissive=Multiply(A=Multiply(A=@n8, B=@n1.A), B=ScalarParameter() {defaultValue=1, parameterName=Opacity})) {materialFunction=&MF_FX_EmissiveAdjustments}
 ]
}
