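# Unlit additive Niagara sprite material for the glow effect on the Pink Spike glider.
# A sphere pin-glow mask's red channel is tinted by particle color and scaled by a view-angle fade.
# The fade uses the camera vector in local space, dimming views along the X axis from either side and, more steeply, views from below (-Z).
# The result feeds the shared MF_FX_EmissiveAdjustments function.
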Material /BRCosmetics/Gadgets/Assets/VinderTech_GliderChute/Glider_PinkSpike/FX/M_Glider_Pinkspike_Glow
graph {
 properties {bUsedWithNiagaraSprites=true, blendMode=BLEND_Additive, pixelDepthOffsetMode=PDOM_Legacy, shadingModel=MSM_Unlit, usedShadingModels=Unlit}
 nodes [
  n1 = TextureSample() {texture=/Game/Global/Textures/Default/Masks_and_Gradients/WM_PinGlow_Sphere}
  n2 = ParticleColor()
  n3 = Transform(None=CameraVectorWS()) {transformSourceType=TRANSFORMSOURCE_World, transformType=TRANSFORM_Local}
  n4 = Constant3Vector() {constant=<rgba 1 0 0 0>}
  n5 = Saturate(None=Abs(None=DotProduct(A=@n3, B=@n4)))
  n6 = DotProduct(
   A=Transform(None=CameraVectorWS()) {transformSourceType=TRANSFORMSOURCE_World, transformType=TRANSFORM_Local}
   B=Constant3Vector() {constant=<rgba 0 0 -1 0>}
  )
  n7 = Multiply(A=Power(Base=OneMinus(None=@n5)) {constExponent=1}, B=Power(Base=OneMinus(None=@n6)) {constExponent=4})
  MaterialFunctionCall(Emissive=Multiply(A=Multiply(A=@n1.R, B=@n2.RGB), B=Saturate(None=@n7))) {materialFunction=&MF_FX_EmissiveAdjustments}
 ]
}
