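# Additive unlit material for Niagara mesh particles that draws the side emissive glow on the PinkSpike glider.
# A grayscale mesh glow texture is shaped by a power and multiplied by a view-angle fade with parameter-driven sensitivity.
# The saturated mask is tinted by particle color and passed through the shared FX emissive adjustments function.
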Material /BRCosmetics/Gadgets/Assets/VinderTech_GliderChute/Glider_PinkSpike/FX/M_Glider_Pinkspike_Side_Emissive
graph {
 properties {
  bUsedWithNiagaraMeshParticles=true
  blendMode=BLEND_Additive
  pixelDepthOffsetMode=PDOM_Legacy
  shadingModel=MSM_Unlit
  usedShadingModels=Unlit
 }
 nodes [
  n1 = TextureSample() {
   samplerSource=SSM_Clamp_WorldGroupSettings
   samplerType=SAMPLERTYPE_LinearGrayscale
   texture=&/T_Glider_PinkSpike_MeshGlow
  }
  n2 = Multiply(
   A=Power(Base=@n1.R)
   B=MaterialFunctionCall("FadeSensitivity (optional)"=ScalarParameter() {parameterName=FadeBasedOnViewAngle}) {materialFunction=&FadeBasedOnViewAngle}.Result
  )
  MaterialFunctionCall(Emissive=Multiply(A=Saturate(None=@n2), B=ParticleColor().RGB)) {materialFunction=&MF_FX_EmissiveAdjustments}
 ]
}
